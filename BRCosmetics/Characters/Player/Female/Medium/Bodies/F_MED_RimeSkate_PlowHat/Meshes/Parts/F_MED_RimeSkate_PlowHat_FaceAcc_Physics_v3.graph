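# Physics asset for the RimeSkate PlowHat female face accessory, defining capsule bodies for dangling hat-hair motion and the surrounding head and upper body.
# Left, middle and right hairHat chains run from BASE bones through segments 06, 04 and 05 with limited swing and twist.
# Cross constraints tie the side 04 segments to the middle one, linking the strands.
# Free-angular constraints connect the BASE bones to the head, and the head and clavicles to spine_05.
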PhysicsAsset /BRCosmetics/Characters/Player/Female/Medium/Bodies/F_MED_RimeSkate_PlowHat/Meshes/Parts/F_MED_RimeSkate_PlowHat_FaceAcc_Physics_v3
bodies {
 BASE_hairHat_L1_06 [{shapeType=Capsule, length=0.505, radius=0.505, shapeName=BASE_hairHat_L1_06_capsule, rotation=<rot 0 -90 0>}]
 BASE_hairHat_M_06 [{shapeType=Capsule, length=0.505, radius=0.505, shapeName=BASE_hairHat_M_06_capsule, rotation=<rot 0 -90 0>}]
 BASE_hairHat_R1_06 [{shapeType=Capsule, length=0.505, radius=0.505, shapeName=BASE_hairHat_R1_06_capsule, rotation=<rot 0 -90 0>}]
 extra_clavicle_l [{shapeType=Capsule, length=9.4996, radius=5.6582, shapeName=clavicle_l_capsule, center=<4.5361 -1.2979 -0.9233>, rotation=<rot 0 -90 88.305>}]
 extra_clavicle_r [{shapeType=Capsule, length=9.4996, radius=5.6582, shapeName=clavicle_r_capsule, center=<-4.5361 1.2979 0.9233>, rotation=<rot 0 -90 -91.695>}]
 hairHat_L1_04 [{shapeType=Capsule, length=4.8335, radius=3.976, shapeName=hair_R1_04_capsule, center=<3.673 0 -3.1818>, rotation=<rot 90 91.1851 0>}]
 hairHat_L1_05 [{
   shapeType=Capsule
   length=7.0544
   radius=5.7756
   shapeName=hair_R1_05_capsule
   center=<7.8046 0.0563 -3.2053>
   rotation=<rot 78.5624 -168.2731 -172.2763>
  }]
 hairHat_L1_06 [{
   shapeType=Capsule
   length=3.9142
   radius=4.2194
   shapeName=hair_R1_03_capsule
   center=<3.7588 0.0279 -3.9228>
   rotation=<rot 5.6499 -177.9425 -86.3442>
  }]
 hairHat_M_04 [{
   shapeType=Capsule
   length=6.563
   radius=3.7932
   shapeName=hair_R1_04_capsule
   center=<3.6725 0.0253 -1.818>
   rotation=<rot 88.4903 -88.8175 179.9974>
  }]
 hairHat_M_05 [{
   shapeType=Capsule
   length=6.635
   radius=6.1951
   shapeName=hair_R1_05_capsule
   center=<7.7648 0.7737 -3.3519>
   rotation=<rot 78.5623 -168.2732 -173.9257>
  }]
 hairHat_M_06 [{
   shapeType=Capsule
   length=5.8473
   radius=4.0143
   shapeName=hair_R1_03_capsule
   center=<4.0463 -0.1443 -1.0806>
   rotation=<rot 5.6365 178.4953 -90.7105>
  }]
 hairHat_R1_04 [{
   shapeType=Capsule
   length=3.5234
   radius=3.5112
   shapeName=hair_R1_04_capsule
   center=<4.9358 -1.0482 -2.5299>
   rotation=<rot 73.1202 119.7103 33.7003>
  }]
 hairHat_R1_05 [{
   shapeType=Capsule
   length=7.4853
   radius=5.3448
   shapeName=hair_R1_05_capsule
   center=<7.6604 0.0264 -2.4774>
   rotation=<rot 78.5624 -168.2731 -172.2763>
  }]
 hairHat_R1_06 [{
   shapeType=Capsule
   length=2.5805
   radius=3.8639
   shapeName=hair_R1_03_capsule
   center=<4.007 -2.4203 -1.9494>
   rotation=<rot 6.5837 -2.5965 -75.9351>
  }]
 head [{
   shapeType=Capsule
   length=8.8132
   radius=9.139
   shapeName=head_capsule
   center=<7.656 -1.8646 0.0613>
   rotation=<rot -0.3497 -90.0002 -1.2793>
  }]
 spine_03 [{shapeType=Capsule, length=15.5082, radius=10.4367, shapeName=spine_03_capsule, center=<2.8404 0.7671 0>, rotation=<rot 0 -90 0>}]
 spine_05 [{shapeType=Capsule, length=6.7148, radius=15.696, shapeName=spine_05_capsule, center=<0.9705 6.2334 0>, rotation=<rot 0 -90 0>}]
 upperarm_twist_01_l [{shapeType=Capsule, length=13.9458, radius=4.8773, shapeName=clavicle_l_capsule, center=<2.2719 -1.2979 0.6337>, rotation=<rot 0 -90 88.305>}]
 upperarm_twist_01_r [{shapeType=Capsule, length=14.8074, radius=4.0158, shapeName=clavicle_r_capsule, center=<-2.3148 0.2123 0.817>, rotation=<rot 0 -90 -91.695>}]
}
constraints [
 {bone1Name=hairHat_L1_06, bone2Name=BASE_hairHat_L1_06, swing1LimitDegrees=20, swing2LimitDegrees=80, twistLimitDegrees=10}
 {bone1Name=hairHat_L1_04, bone2Name=hairHat_L1_06, swing1LimitDegrees=10, swing2LimitDegrees=20, twistLimitDegrees=5}
 {bone1Name=hairHat_L1_05, bone2Name=hairHat_L1_04, swing1LimitDegrees=10, swing2LimitDegrees=20, twistLimitDegrees=5}
 {bone1Name=hairHat_M_06, bone2Name=BASE_hairHat_M_06, swing1LimitDegrees=20, swing2LimitDegrees=80, twistLimitDegrees=10}
 {bone1Name=hairHat_M_04, bone2Name=hairHat_M_06, swing1LimitDegrees=20, swing2LimitDegrees=40, twistLimitDegrees=10}
 {bone1Name=hairHat_M_05, bone2Name=hairHat_M_04, swing1LimitDegrees=10, swing2LimitDegrees=20, twistLimitDegrees=5}
 {bone1Name=hairHat_R1_06, bone2Name=BASE_hairHat_R1_06, swing1LimitDegrees=10, swing2LimitDegrees=80, twistLimitDegrees=5}
 {bone1Name=hairHat_R1_04, bone2Name=hairHat_R1_06, swing1LimitDegrees=10, swing2LimitDegrees=20, twistLimitDegrees=5}
 {bone1Name=hairHat_R1_05, bone2Name=hairHat_R1_04, swing1LimitDegrees=10, swing2LimitDegrees=20, twistLimitDegrees=5}
 {bone1Name=hairHat_R1_04, bone2Name=hairHat_M_04, swing1LimitDegrees=30, swing2LimitDegrees=15, twistLimitDegrees=15}
 {bone1Name=hairHat_L1_04, bone2Name=hairHat_M_04, swing1LimitDegrees=30, swing2LimitDegrees=15, twistLimitDegrees=15}
 {
  bone1Name=extra_clavicle_l
  bone2Name=spine_05
  swing1LimitDegrees=45
  swing1Motion=Free
  swing2LimitDegrees=45
  swing2Motion=Free
  twistLimitDegrees=45
  twistMotion=Free
 }
 {
  bone1Name=extra_clavicle_r
  bone2Name=spine_05
  swing1LimitDegrees=45
  swing1Motion=Free
  swing2LimitDegrees=45
  swing2Motion=Free
  twistLimitDegrees=45
  twistMotion=Free
 }
 {
  bone1Name=BASE_hairHat_L1_06
  bone2Name=head
  swing1LimitDegrees=45
  swing1Motion=Free
  swing2LimitDegrees=45
  swing2Motion=Free
  twistLimitDegrees=45
  twistMotion=Free
 }
 {
  bone1Name=BASE_hairHat_R1_06
  bone2Name=head
  swing1LimitDegrees=45
  swing1Motion=Free
  swing2LimitDegrees=45
  swing2Motion=Free
  twistLimitDegrees=45
  twistMotion=Free
 }
 {
  bone1Name=BASE_hairHat_M_06
  bone2Name=head
  swing1LimitDegrees=45
  swing1Motion=Free
  swing2LimitDegrees=45
  swing2Motion=Free
  twistLimitDegrees=45
  twistMotion=Free
 }
 {
  bone1Name=head
  bone2Name=spine_05
  swing1LimitDegrees=45
  swing1Motion=Free
  swing2LimitDegrees=45
  swing2Motion=Free
  twistLimitDegrees=45
  twistMotion=Free
 }
]
