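# Pre-physics Control Rig for the female RimeSkate PlowHat face accessory, adjusting extra clavicle bones and under-hat hair bases.
# A sequence first parent-constrains each extra clavicle to its BASE bone, weighting the right only partially; the two clavicle pose readers feed nothing.
# It then reads head pose against neck_01 with a spherical pose reader, whose smoothed output weights two stacked transform offsets on BASE_hairHat bones.
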ControlRigBlueprint /BRCosmetics/Characters/Player/Female/Medium/Bodies/F_MED_RimeSkate_PlowHat/Meshes/Parts/F_MED_RimeSkate_PlowHat_FaceAcc_PrePhysics_CtrlRig
graphs [
 RigVMFunctionLibrary {links [], nodes []}
 RigVMModel {nodes [
   RigUnit_BeginExecution = &RigUnit_BeginExecution()
   RigVMFunction_Sequence = &RigVMFunction_Sequence(ExecuteContext=@RigUnit_BeginExecution.ExecutePin)
   &RigUnit_SphericalPoseReader(
    ActiveRegionScaleFactors=(PositiveWidth=1,NegativeWidth=0.1,PositiveHeight=1,NegativeHeight=0.5)
    ActiveRegionSize=0.3
    Debug=(bDrawDebug=False,bDraw2D=false,bDrawLocalAxes=false,DebugScale=30,DebugSegments=30,DebugThickness=0.2)
    DriverAxis=(X=1,Y=0,Z=0)
    DriverItem=(Type=Bone,Name="clavicle_l")
    FalloffRegionScaleFactors=(PositiveWidth=1,NegativeWidth=0.5,PositiveHeight=1,NegativeHeight=0.5)
    FalloffSize=0.5
    OptionalParentItem=(Type=Bone,Name="None")
    RotationOffset=(X=0,Y=40,Z=180)
    ExecutePin=@RigVMFunction_Sequence.A
   )
   > &RigUnit_SphericalPoseReader(
    ActiveRegionScaleFactors=(PositiveWidth=1,NegativeWidth=0.1,PositiveHeight=0.5,NegativeHeight=1)
    ActiveRegionSize=0.3
    Debug=(bDrawDebug=False,bDraw2D=false,bDrawLocalAxes=false,DebugScale=30,DebugSegments=30,DebugThickness=0.2)
    DriverAxis=(X=-1,Y=0,Z=0)
    DriverItem=(Type=Bone,Name="clavicle_r")
    FalloffRegionScaleFactors=(PositiveWidth=1,NegativeWidth=0.5,PositiveHeight=1,NegativeHeight=1)
    FalloffSize=0.5
    OptionalParentItem=(Type=Bone,Name="None")
    RotationOffset=(X=0,Y=220,Z=180)
   )
   > &RigUnit_ParentConstraint(
    AdvancedSettings=(InterpolationType=Average,RotationOrderForFilter=XZY)
    Child=(Type=Bone,Name="extra_clavicle_l")
    Parents=((Item=(Type=Bone,Name="BASE_clavicle_l"),Weight=1))
    Weight=1
    bMaintainOffset=false
   )
   > &RigUnit_ParentConstraint(
    AdvancedSettings=(InterpolationType=Average,RotationOrderForFilter=XZY)
    Child=(Type=Bone,Name="extra_clavicle_r")
    Parents=((Item=(Type=Bone,Name="BASE_clavicle_r"),Weight=1))
    Weight=0.2
    bMaintainOffset=false
   )
   n1 = &RigUnit_SphericalPoseReader(
    ActiveRegionScaleFactors=(PositiveWidth=0.2,NegativeWidth=2,PositiveHeight=2,NegativeHeight=2)
    ActiveRegionSize=0.1
    Debug=(bDrawDebug=True,bDraw2D=False,bDrawLocalAxes=False,DebugScale=25,DebugSegments=20,DebugThickness=0.25)
    DriverAxis=(X=1,Y=0,Z=0)
    DriverItem=(Type=Bone,Name="head")
    FalloffRegionScaleFactors=(PositiveWidth=0.3,NegativeWidth=3,PositiveHeight=3,NegativeHeight=3)
    FalloffSize=0.2
    FlipHeightScaling=False
    FlipWidthScaling=False
    OptionalParentItem=(Type=Bone,Name="neck_01")
    RotationOffset=(X=-90,Y=0,Z=70)
    ExecutePin=@RigVMFunction_Sequence.B
   )
   n2 = &RigVMFunction_AlphaInterp(InterpSpeedDecreasing=10, InterpSpeedIncreasing=10, Scale=1, bClampResult=False, Value=@n1.OutputParam)
   > &RigUnit_ModifyTransforms(
    ItemToModify="((Item=(Type=Bone,Name=\"BASE_hairHat_L1_06\"),Transform=(Rotation=(X=0,Y=0,Z=0,W=1),Translation=(X=0,Y=0,Z=-2),Scale3D=(X=1,Y=1,Z=1))),(Item=(Type=Bone,Name=\"BASE_hairHat_M_06\"),Transform=(Rotation=(X=0,Y=0,Z=0,W=1),Translation=(X=0,Y=0,Z=-1),Scale3D=(X=1,Y ...(truncated)"
    WeightMinimum=0
    Weight=@n2.Result
    ItemToModify.0.Transform.Rotation=&RigVMFunction_MathQuaternionFromEuler(Euler=(X=0,Y=0,Z=0), RotationOrder=XYZ).Result
    ItemToModify.1.Transform.Rotation=&RigVMFunction_MathQuaternionFromEuler(Euler=(X=0,Y=0,Z=0), RotationOrder=XYZ).Result
    ItemToModify.2.Transform.Rotation=&RigVMFunction_MathQuaternionFromEuler(Euler=(X=0,Y=0,Z=0), RotationOrder=XYZ).Result
   )
   > &RigUnit_ModifyTransforms(
    ItemToModify="((Item=(Type=Bone,Name=\"BASE_hairHat_L1_06\"),Transform=(Rotation=(X=0,Y=0,Z=0,W=1),Translation=(X=0,Y=0,Z=-2),Scale3D=(X=1,Y=1,Z=1))),(Item=(Type=Bone,Name=\"BASE_hairHat_M_06\"),Transform=(Rotation=(X=0,Y=0,Z=0,W=1),Translation=(X=0,Y=0,Z=0),Scale3D=(X=1,Y= ...(truncated)"
    WeightMinimum=0
    Weight=@n2.Result
    ItemToModify.0.Transform.Rotation=&RigVMFunction_MathQuaternionFromEuler(Euler=(X=0,Y=0,Z=0), RotationOrder=XYZ).Result
    ItemToModify.1.Transform.Rotation=&RigVMFunction_MathQuaternionFromEuler(Euler=(X=0,Y=0,Z=0), RotationOrder=XYZ).Result
    ItemToModify.2.Transform.Rotation=&RigVMFunction_MathQuaternionFromEuler(Euler=(X=0,Y=0,Z=0), RotationOrder=XYZ).Result
   )
  ]}
]
hierarchy [
 root {loc=<0 0 0>, rot=<0 0 0 1>}
 pelvis {parent=root, loc=<0 -0.5301 98.0571>, rot=<0.0127 0.707 -0.0127 -0.707>}
 spine_01 {parent=pelvis, loc=<6.1208 -0.9328 0>, rot=<0 0 0.034 -0.9994>}
 spine_02 {parent=spine_01, loc=<6.5997 -0.2095 0>, rot=<0 0 0.0238 -0.9997>}
 spine_03 {parent=spine_02, loc=<6.4914 -0.1237 0>, rot=<0 0 0.0149 -0.9999>}
 spine_04 {parent=spine_03, loc=<6.9598 0.0794 0>, rot=<0 0 -0.0132 -0.9999>}
 spine_05 {parent=spine_04, loc=<6.9987 0 0>, rot=<0 0 -0.0015 -1>}
 neck_01 {parent=spine_05, loc=<11.3138 0.1529 0>, rot=<0 0 -0.2003 -0.9797>}
 neck_02 {parent=neck_01, loc=<4.5627 0 0>, rot=<0 0 0.0504 -0.9987>}
 head {parent=neck_02, loc=<5.3946 -0.0007 -0.0012>, rot=<0 0.0002 0.1145 -0.9934>}
 hair_L1_01 {parent=head, loc=<4.1073 -3.7482 -4.1059>, rot=<-0.0677 0.5148 0.8546 -0.0054>}
 hair_L1_02 {parent=hair_L1_01, loc=<3.4065 0 0>, rot=<0 0 0 1>}
 hair_L1_03 {parent=hair_L1_02, loc=<6.4688 0 0>, rot=<0 -0.3146 0 -0.9492>}
 hair_L1_04 {parent=hair_L1_03, loc=<6.4688 0 0>, rot=<0 0.1502 0 -0.9887>}
 hair_L1_05 {parent=hair_L1_04, loc=<6.4688 0 0>, rot=<0 0.1746 0 -0.9846>}
 hair_R1_01 {parent=head, loc=<3.9503 -3.5005 3.7489>, rot=<0.0156 -0.8643 -0.4937 0.0955>}
 hair_R1_02 {parent=hair_R1_01, loc=<3.4065 0 0>, rot=<0 0 0 1>}
 hair_R1_03 {parent=hair_R1_02, loc=<5.5304 0 0>, rot=<-0.0648 -0.2679 0.0442 -0.9603>}
 hair_R1_04 {parent=hair_R1_03, loc=<6.3656 0 0>, rot=<0.029 0.1096 -0.0012 -0.9936>}
 hair_R1_05 {parent=hair_R1_04, loc=<7.9626 0 0>, rot=<0 0.2157 0 -0.9765>}
 hair_M_01 {parent=head, loc=<3.9056 -6.8385 0.3428>, rot=<-0.1094 0.6866 0.7112 -0.1038>}
 hair_M_02 {parent=hair_M_01, loc=<3.4065 0 0>, rot=<0 0 0 1>}
 hair_M_03 {parent=hair_M_02, loc=<4.6411 0 0>, rot=<0 0 0 1>}
 hair_M_04 {parent=hair_M_03, loc=<6.4688 0 0>, rot=<0 0 0 1>}
 hair_M_05 {parent=hair_M_04, loc=<6.4688 0 0>, rot=<0.0033 0.1446 0.0559 -0.9879>}
 lock_L_01 {parent=head, loc=<3.6901 1.8733 -6.4436>, rot=<0.0104 -0.707 -0.707 0.0146>}
 lock_L_02 {parent=lock_L_01, loc=<4.0206 0 0>, rot=<0.0027 0.0248 -0.1014 -0.9945>}
 lock_L_03 {parent=lock_L_02, loc=<3.8733 0 0>, rot=<0 0 0 1>}
 lock_R_01 {parent=head, loc=<3.2969 1.808 6.8232>, rot=<0.0434 -0.7795 -0.6243 -0.0279>}
 lock_R_02 {parent=lock_R_01, loc=<4.6799 0 0>, rot=<-0.0138 -0.0142 0.0928 -0.9955>}
 lock_R_03 {parent=lock_R_02, loc=<2.3226 0 0>, rot=<0 0 0 1>}
 earring_l {parent=head, loc=<0.5327 -0.846 -6.5156>, rot=<0.0038 -0.9813 -0.1925 -0.0005>}
 earring_r {parent=head, loc=<0.5327 -0.8449 6.5153>, rot=<0.0038 -0.9873 0.1589 0.0006>}
 strand_a_01 {parent=head, loc=<9.7348 7.7973 -2.5027>, rot=<-0.0674 -0.8725 -0.2691 -0.4022>}
 strand_a_02 {parent=strand_a_01, loc=<2.2564 0 0>, rot=<0 -0.3456 0 -0.9384>}
 strand_b_01 {parent=head, loc=<8.3685 7.0768 -4.0531>, rot=<0.2576 -0.8948 -0.2901 -0.2208>}
 strand_c_01 {parent=head, loc=<5.1708 5.0573 -6.0293>, rot=<0.0484 -0.9982 -0.0032 -0.0357>}
 BASE_hairHat_L1_06 {parent=head, loc=<-4.4231 -3.4231 -4.5183>, rot=<-0.3331 0.487 0.7899 -0.1671>}
 hairHat_L1_06 {parent=BASE_hairHat_L1_06, loc=<0 0 0>, rot=<0 0 0 1>}
 hairHat_L1_04 {parent=hairHat_L1_06, loc=<8.2696 0 -0.068>, rot=<0 0.1502 0 -0.9887>}
 hairHat_L1_05 {parent=hairHat_L1_04, loc=<6.4688 0 0>, rot=<0 0.1746 0 -0.9846>}
 BASE_hairHat_R1_06 {parent=head, loc=<-4.1974 -4.108 4.8018>, rot=<0.1916 -0.8356 -0.4181 0.3004>}
 hairHat_R1_06 {parent=BASE_hairHat_R1_06, loc=<0 0 0>, rot=<0 0 0 1>}
 hairHat_R1_04 {parent=hairHat_R1_06, loc=<7.199 0 0>, rot=<0.029 0.1096 -0.0012 -0.9936>}
 hairHat_R1_05 {parent=hairHat_R1_04, loc=<7.9626 0 0>, rot=<0 0.2157 0 -0.9765>}
 BASE_hairHat_M_06 {parent=head, loc=<-2.0746 -8.7047 0.2607>, rot=<-0.1094 0.6866 0.7112 -0.1038>}
 hairHat_M_06 {parent=BASE_hairHat_M_06, loc=<0 0 0>, rot=<0 0 0 1>}
 hairHat_M_04 {parent=hairHat_M_06, loc=<8.2513 0 0>, rot=<0 0 0 1>}
 hairHat_M_05 {parent=hairHat_M_04, loc=<6.4688 0 0>, rot=<0.0033 0.1446 0.0559 -0.9879>}
 clavicle_l {parent=spine_05, loc=<5.4831 -0.1053 -3.2946>, rot=<0.0465 -0.7289 -0.009 -0.6829>}
 upperarm_l {parent=clavicle_l, loc=<9.8658 -1.2136 -0.4105>, rot=<0.0164 -0.3245 -0.0269 -0.9454>}
 upperarm_twist_01_l {parent=upperarm_l, loc=<7.2135 -0.1134 0>, rot=<0 0 0 1>}
 deform_upperarm_bk_base_l {parent=upperarm_twist_01_l, loc=<-6.4053 -4.3746 0.314>, rot=<0 0 0 -1>}
 deform_upperarm_bk_l {parent=deform_upperarm_bk_base_l, loc=<0 0 0>, rot=<0 0 0 1>}
 deform_upperarm_fr_base_l {parent=upperarm_twist_01_l, loc=<-5.6312 5.092 0.3354>, rot=<0 0 0 -1>}
 deform_upperarm_fr_l {parent=deform_upperarm_fr_base_l, loc=<0 0 0>, rot=<0 0 0 1>}
 deform_upperarm_up_base_l {parent=upperarm_twist_01_l, loc=<-8.891 0.4825 3.9517>, rot=<0 0 0 -1>}
 deform_upperarm_up_l {parent=deform_upperarm_up_base_l, loc=<0 0 0>, rot=<0 0 0 1>}
 extra_clavicle_l {parent=clavicle_l, loc=<0 0 0>, rot=<0 0 0 1>}
 clavicle_r {parent=spine_05, loc=<5.4831 -0.1053 3.2946>, rot=<0.7289 0.0465 -0.6829 0.009>}
 upperarm_r {parent=clavicle_r, loc=<-9.8658 1.2136 0.4105>, rot=<0.0164 -0.3245 -0.0269 -0.9454>}
 upperarm_twist_01_r {parent=upperarm_r, loc=<-7.2135 0.1134 0>, rot=<0 0 0 1>}
 deform_upperarm_bk_base_r {parent=upperarm_twist_01_r, loc=<6.4053 4.3746 -0.314>, rot=<0 0 0 -1>}
 deform_upperarm_bk_r {parent=deform_upperarm_bk_base_r, loc=<0 0 0>, rot=<0 0 0 1>}
 deform_upperarm_fr_base_r {parent=upperarm_twist_01_r, loc=<5.6312 -5.092 -0.3354>, rot=<0 0 0 -1>}
 deform_upperarm_fr_r {parent=deform_upperarm_fr_base_r, loc=<0 0 0>, rot=<0 0 0 1>}
 deform_upperarm_up_base_r {parent=upperarm_twist_01_r, loc=<8.891 -0.4825 -3.9517>, rot=<0 0 0 -1>}
 deform_upperarm_up_r {parent=deform_upperarm_up_base_r, loc=<0 0 0>, rot=<0 0 0 1>}
 extra_clavicle_r {parent=clavicle_r, loc=<0 0 0>, rot=<0 0 0 1>}
 BASE_clavicle_r {parent=spine_05, loc=<5.4831 -0.1053 3.2946>, rot=<0.7289 0.0465 -0.6829 0.009>}
 BASE_clavicle_l {parent=spine_05, loc=<5.4831 -0.1053 -3.2946>, rot=<0.0465 -0.7289 -0.009 -0.6829>}
]
