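# Animation Blueprint for the female RimeSkate PlowHat face accessory, adding secondary physics to the cosmetic part.
# AnimGraph copies the attached parent's pose through a pre-physics Control Rig, head-space RigidBody simulation and post-physics Control Rig.
# Each update stores gravity and rigid-body settings computed from the Plow variant's parameter asset, plus backbling and cape flags that feed the post-physics rig.
# Customization load caches a body mesh reference.
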AnimBlueprint /BRCosmetics/Characters/Player/Female/Medium/Bodies/F_MED_RimeSkate_PlowHat/Meshes/Parts/F_MED_RimeSkate_PlowHat_FaceAcc_AnimBP
target_skeleton=&/F_MED_RimeSkate_PlowHat_FaceAcc_Skeleton
graphs [
 AnimGraph {nodes [
   n1 = Control Rig(Source=Copy Pose From Mesh() {bUseAttachedParent=true}) {controlRigAssetReference {blueprintRigClass=&/F_MED_RimeSkate_PlowHat_FaceAcc_PrePhysics_CtrlRig.F_MED_RimeSkate_PlowHat_FaceAcc_PrePhysics_CtrlRig_C}}
   n2 = RigidBody!(
    ComponentPose=Local To Component(LocalPose=@n1)
    bAlphaBoolEnabled:bool
    OverrideWorldGravity:Vector="0,0,-980"
    SimSpaceSettings:Sim Space Settings Structure=(WorldAlpha=0,VelocityScaleZ=1,DampingAlpha=1,MaxLinearVelocity=10000,MaxAngularVelocity=10000,MaxLinearAcceleration=10000,MaxAngularAcceleration=10000,ExternalLinearDragV=(X=0,Y=0,Z=0),ExternalLinearVelocity=(X=0,Y=0,Z=0),ExternalAngularVelocity=(X=0,Y=0,Z=0))
   ) -> (Pose) {
    alphaInputType=Bool
    bClampLinearTranslationLimitToRefPose=true
    bOverrideWorldGravity=true
    overridePhysicsAsset=&/F_MED_RimeSkate_PlowHat_FaceAcc_Physics_v3
    simulationSpace=BaseBoneSpace
    alphaBoolBlend {blendInTime=0.2, blendOption=Linear, blendOutTime=0.2}
    baseBoneRef {boneName=head}
    overrideWorldGravity {z=-980}
    simSpaceSettings {
     maxAngularAcceleration=65
     maxAngularVelocity=1500
     maxLinearAcceleration=1000
     maxLinearVelocity=3000
     velocityScaleZ=0.3
     worldAlpha=1
     externalLinearDragV=<1 1 1>
    }
   }
   n3 = Control Rig!(
    Source=Component To Local(ComponentPose=@n2)
    Alpha
    backpack_cape_mod_bool:bool=Get backpack_cape_mod_bool() -> (backpack_cape_mod_bool:bool)
    RigidBodyEnabled:bool=Get bRigidBodyEnabled()
    cape_mod:bool=Get cape_mod() -> (cape_mod:bool)
   ) -> (Pose) {controlRigAssetReference {blueprintRigClass=&/F_MED_RimeSkate_PlowHat_FaceAcc_PostPhysics_CtrlRig.F_MED_RimeSkate_PlowHat_FaceAcc_PostPhysics_CtrlRig_C}}
   Output Pose(Result=@n3) {layerGroup=None}
  ]}
 EventGraph {nodes [
   Event BlueprintUpdateAnimation()
   > n1 = "Gravity Override State Function (Parameter Asset) (Nativized)"(
    CharacterPartAnimInstance=Self-Reference()
    GameTimeInSeconds=GetGameTimeInSeconds()
    Parameters=/BRCosmetics/Characters/Player/Female/Medium/Bodies/F_MED_RimeSkate_Plow/Meshes/Parts/F_MED_RimeSkate_Plow_FaceAcc_Parameters
   )
   > Set RBAN Settings Out(RBAN Settings Out=@n1.OutRigidBodyAnimNodeParameters)
   > Set Gravity Override(Gravity Override=@n1)
   n2 = CheckForBackbling()
   > Set backpack_cape_mod_bool(backpack_cape_mod_bool:bool=OR Boolean(A=@n2, B=@n2.bIsBackblingACape)) -> (Output_Get:bool)
   > Set cape_mod(cape_mod:bool=@n2.bIsBackblingACape) -> (Output_Get:bool)
   n3 = Event OnCharacterCustomizationLoaded()
   > Set BodySkelMesh(BodySkelMesh:Skeletal Mesh Component Budgeted Object Reference=Get Array Item with EFort Custom Part Type(Array=@n3.CustomMeshParts)) -> (Output_Get:Skeletal Mesh Component Budgeted Object Reference)
  ]}
]
variables [
 backpack_cape_mod_bool {type=boolean}
 bind_pose {type=boolean, default=true}
 BodySkelMesh {type=SkeletalMeshComponentBudgeted}
 cape_mod {type=boolean}
 Gravity Override {type=Vector}
 RBAN Settings Out {type=FortRigidBodyAnimNodeParameters}
 Sim Space Settings {
  type=SimSpaceSettings
  default {
   dampingAlpha=1
   maxAngularAcceleration=10000
   maxAngularVelocity=10000
   maxLinearAcceleration=10000
   maxLinearVelocity=10000
   velocityScaleZ=1
   worldAlpha=0
   externalLinearDragV=<0 0 0>
  }
 }
]
